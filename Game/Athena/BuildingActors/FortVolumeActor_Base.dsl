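; FortVolumeActor_Base is a base volume actor Blueprint that shows a bounds mesh which fades with the actor's lifetime.
; The construction script initializes the bounds component.
; On BeginPlay, where cosmetic events can run, it creates a dynamic material from the source material and plays TransitionTL to drive FadeInBias.
; The death effects event reverses that timeline, which hides the mesh when it finishes backward.
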(event EventBeginPlay
  (|CallParentFunction)
  (Networking|CanExecuteCosmeticEvents
    (:True
      (Utilities|FlowControl|Sequence
        (:then_0
          ;[1 1 1] Initialize MID
          (bind _boundsmeshsourcematerial (Variables|Default|GetBoundsMeshSourceMaterial)
            (:then
              (bind _returnvalue (Rendering|Material|CreateDynamicMaterialInstance (Variables|Default|GetBoundsMeshComponent) 0 _boundsmeshsourcematerial "None" :!class /Script/Engine.PrimitiveComponent))
              (Variables|Default|SetBoundsMeshMID _returnvalue))
            (:else)))
        (:then_1
          (bind (_direction _fade) (Timeline|TransitionTL PlayFromStart)
            (:Update
              (Rendering|Material|SetScalarParameterValue (Variables|Default|GetBoundsMeshMID) "FadeInBias" _fade :!class /Script/Engine.MaterialInstanceDynamic))
            (:Finished
              (switch Utilities|FlowControl|Switch|SwitchonETimelineDirection _direction
                (:Backward
                  (Rendering|SetVisibility (Variables|Default|GetBoundsMeshComponent) false false))))))))))

(event Custom|OnDeathPlayEffects_1
  (params (Damage :type Float%20%28double-precision%29) (DamageTags :type Gameplay%20Tag%20Container%20Structure%20%28by%20ref%29) (Momentum :type Vector) (HitInfo :type Hit%20Result%20Structure%20%28by%20ref%29) (InstigatedBy :type Fort%20Pawn%20Object%20Reference) (DamageCauser :type Actor%20Object%20Reference) (EffectContext :type Gameplay%20Effect%20Context%20Handle%20Structure))
  (Timeline|TransitionTL ReverseFromEnd))

(fn ConstructionScript
  (|CallParentFunction)
  (Volume|InitializeBoundsComponent (Variables|Default|GetBoundsMeshComponent)))
